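# Niagara dust effect for the ForestJeanGrowl emote, with one sprite emitter of small, multicolored glitter in pink, purple and dark tones.
# Integer changes in User.BurstCurve trigger small bursts, while User.BurstIndex picks one of two mirrored spawn offsets.
# Particles spray in a cone under gravity, drag and curl noise, with brightness spikes, flicker and time-of-day tinting.
# BlendSpace varies world-to-local blending by owner speed against a walk-speed threshold.
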NiagaraSystem /BRCosmetics/Animation/Game/MainPlayer/Emotes/ForestJeanGrowl/VFX/NS_Emote_ForestJeanGrowl_Dust
dependencies {
 usedDataInterfaces [&NiagaraDataInterfaceArrayColor, &NiagaraDataInterfaceCurve, &NiagaraDataInterfaceColorCurve]
 usedDynamicInputs []
 usedModules [&SystemState, &BlendSpaceHelper]
}
emitters {glitter {
  emitterUpdateScript [
   &EmitterState()
   /Game/Effects/Niagara/Modules/Logic/UpdateValueStateChange(InputIntParam:NiagaraInt32=&MakeIntFromFloat(Float=$User.BurstCurve, Conversion Mode=::NewEnumerator1))
   &self:TriggerSpawn(
    Spawn Count:NiagaraInt32=&RandomRangeInt(Minimum=11, Maximum=13)
    Spawn Group:NiagaraInt32=&MakeIntFromFloat(Float=0, Conversion Mode=::NewEnumerator1)
    Trigger:NiagaraBool=$Emitter.UpdateValueStateChange.Changed
   ) {module=TriggerSpawn}
   /SetVariables(Emitter.WalkForwardSpeed=145) {isSetParameters=true}
  ]
  particleSpawnScript [
   &InitializeParticle(
    Color=&SelectLinearColorFromArray(
     Array Sampling Mode=::NewEnumerator0
     Direct Array Index=?
     Color Selection Array=DataInterface() {propertyValues {ColorData [
        <rgba 1 0.0714 0.5244 1>
        <rgba 0.5687 0.0571 1 1>
        <rgba 0.1473 0.1405 0.149 1>
        <rgba 0.8187 0.1885 1 1>
        <rgba 0.1226 0.0724 0.2406 1>
        <rgba 0.4133 0.1594 1 1>
        <rgba 0.7698 0.0003 0.6263 1>
        <rgba 0.4281 0 0.062 1>
       ]}}
     Random Seed=0
    )
    Lifetime Max=1
    Lifetime Min=0.5
    Lifetime Mode=::NewEnumerator1
    Mass Max=0.7
    Mass Min=0.2
    Mass Mode=::NewEnumerator3
    Sprite Size Mode=::NewEnumerator1
    Uniform Sprite Size Max=6
    Uniform Sprite Size Min=1.5
    UsePositionOffset=false
    Write Material Random=true
   )
   &AddVelocity(
    Cone Angle=50
    Cone Axis=<0 -1 0.8>
    Distribution Along Cone Axis=0.2
    Module.UseConeFalloff=false
    Module.UseVelDistribution=true
    Override Seed=false
    Velocity Mode=::NewEnumerator2
    Velocity Speed=&RandomRangeFloat(Minimum=70, Maximum=300)
   )
   &ShapeLocation(
    MaskSpawnGroup=false
    Offset=&MakeCustomVectorFromBool(Bool=&SetBoolByFloatComparison(A=$User.BurstIndex, B=0.5), True Vector=<3 -6 0>, False Vector=<-3 -6 0>)
    Offset Mode=::NewEnumerator0
    Override Seed=false
    Sphere Radius=3
    UseOwnerScale=true
   )
  ]
  particleUpdateScript [
   &ParticleState()
   &ScaleSpriteSize(
    Uniform Curve Scale=&RandomRangeFloat(Minimum=0.5, Maximum=1)
    Uniform Curve Sprite Scale=DataInterface() {propertyValues {Curve {
       defaultValue=0
       keys [
        {interpMode=RCIM_Cubic, tangentMode=RCTM_User, time=0, value=0}
        {interpMode=RCIM_Cubic, tangentMode=RCTM_User, time=0.5, value=1}
        {tangentMode=RCTM_User, time=1, value=0}
       ]
      }}}
   )
   &GravityForce(Gravity=&Multiply_VectorByFloat(
     Vector=<0 0 -300>
     Float=&FloatFromCurve(FloatCurve=DataInterface() {propertyValues {Curve {
         defaultValue=1
         keys [
          <at 0.1191 0 RCIM_Cubic>
          {arriveTangent=-0.7473, interpMode=RCIM_Cubic, leaveTangent=-0.7473, tangentMode=RCTM_User, time=0.3761, value=0.9394}
         ]
        }}})
    ))
   &Drag(Drag=&RandomRangeFloat(Minimum=0.8, Maximum=1.2), Use Linear Drag=true, Use Rotational Drag=false)
   &ScaleColor(
    Scale Alpha=&FloatFromCurve(
     FloatCurve=DataInterface() {propertyValues {Curve {keys [<at -0.0046 1.0113 RCIM_Cubic>,<at 1.0011 0.002 RCIM_Cubic>]}}}
     Scale Curve=&Waveform("Waveform [1]"=::NewEnumerator0, "[1] Frequency"=4, "[1] Amplitude Min/Max"=<0.1 1.1>)
    )
    Scale RGB=&Multiply_Vector(
     A=&MakeVectorFromLinearColorRGB(Color=&FNTimeOfDayColor(
       Night Color=<rgba 0.3406 0.3406 0.3406 1>
       Morning Color=<rgba 0.6406 0.6406 0.6406 1>
       Day Color=<rgba 1 1 1 1>
       Evening Color=<rgba 0.6406 0.6406 0.6406 1>
      ))
     B=&MakeVectorFromLinearColorRGB(Color=&MultiplyLinearColor(
       Linear Color=&ColorFromCurve(ColorCurve=DataInterface() {propertyValues {
          AlphaCurve {keys []}
          BlueCurve {keys [<at 0 1 RCIM_Cubic>,<at 0.1015 1>,<at 0.118 200>,<at 0.1312 1>,<at 0.306 1>,<at 0.3225 350>,<at 0.3379 1>,<at 1 1 RCIM_Cubic>]}
          GreenCurve {keys [<at 0 1 RCIM_Cubic>,<at 0.1015 1>,<at 0.118 200>,<at 0.1312 1>,<at 0.306 1>,<at 0.3225 350>,<at 0.3379 1>,<at 1 1 RCIM_Cubic>]}
          RedCurve {keys [<at 0 1 RCIM_Cubic>,<at 0.1015 1>,<at 0.118 200>,<at 0.1312 1>,<at 0.306 1>,<at 0.3225 350>,<at 0.3379 1>,<at 1 1 RCIM_Cubic>]}
         }})
       Scale Factor=&MakeVector4_FromLinearColor(LinearColor=&LerpLinearColors(
         StartColor=<rgba 1 1 1 1>
         EndColor=<rgba 20 20 20 1>
         LerpFactor=&Waveform("Waveform [1]"=::NewEnumerator0, "[1] Frequency"=4, "[1] Amplitude Min/Max"=<0 1>)
        ))
      ))
    )
    ScaleA=true
    ScaleRGB=true
   )
   &CurlNoiseForce(
    Bias Noise=false
    Noise Frequency=&RandomRangeFloat(Minimum=30, Maximum=40)
    Noise Strength=100
    Pan Noise=false
   )
   &SolveForcesAndVelocity(Clamp Velocity=false, Limit Acceleration=false)
   &ScaleSpriteSizeBySpeed(
    Max Scale Factor=<1 7>
    Sample Scale Factor By Curve=true
    Scale Factor Curve=DataInterface() {propertyValues {Curve {keys [
        {arriveTangent=2.865, interpMode=RCIM_Cubic, time=0.1247, value=-0.0006}
        {arriveTangent=4.1522, interpMode=RCIM_Cubic, leaveTangent=4.1522, tangentMode=RCTM_User, time=0.3948, value=0.4893}
        <at 0.9958 0.9974 RCIM_Cubic>
       ]}}}
    Velocity Threshold=700
   )
   &BlendSpace("Position Blend (World <-> Local)":NiagaraFloat=&MakeCustomFloatFromBool(
     Bool=&SetBoolByFloatComparison(A=&VectorLength(Vector=$User.OwnerLocalVelocity), B=$Emitter.WalkForwardSpeed)
     True Float=&FloatFromCurve(FloatCurve=DataInterface() {propertyValues {Curve {keys [<at 0.4 1>,<at 0.8 0>]}}})
     False Float=&FloatFromCurve(FloatCurve=DataInterface() {propertyValues {Curve {keys [<at 0 1>,<at 0.0793 0>]}}})
     Translation Contribution:NiagaraFloat=&MakeCustomFloatFromBool(
      Bool=&SetBoolByFloatComparison(A=&VectorLength(Vector=$User.OwnerLocalVelocity), B=$Emitter.WalkForwardSpeed)
      True Float=1
      False Float=0
     )
     Rotation Contribution:NiagaraFloat=&MakeCustomFloatFromBool(
      Bool=&SetBoolByFloatComparison(A=&VectorLength(Vector=$User.OwnerLocalVelocity), B=$Emitter.WalkForwardSpeed)
      True Float=1
      False Float=0
     )
    ))
  ]
  rendererClasses [&NiagaraSpriteRendererProperties]
  renderers [{rendererClass=&NiagaraSpriteRendererProperties}]
 }}
summary {
 systemName=NS_Emote_ForestJeanGrowl_Dust
 emitters [{emitterName=glitter, rendererClasses [&NiagaraSpriteRendererProperties]}]
 userVariables [
  User.bIsFrontEnd {type=NiagaraBool}
  User.bIsFrontEndPreview {type=NiagaraBool}
  User.BurstCurve {type=NiagaraFloat, defaultValue=0, description=""}
  User.BurstIndex {type=NiagaraFloat, defaultValue=0, description=""}
  User.OwnerLocalVelocity {type=Vector3f, description="", defaultValue {struct=Vector3f, value=<0 0 0>}}
 ]
}
